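# Translucent, unlit, two-sided Niagara sprite and mesh particle material for a soft blur effect in the CrumbViolin Baguette emote.
# Emissive comes from a tail blur texture tinted by particle color, passed through a shared emissive adjustment function.
# Opacity combines particle alpha, an alpha multiplier and a centered UV sphere mask with adjustable radius and hardness, then applies depth fade to soften intersections.
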Material /BRCosmetics/Effects/Fort_Effects/Emotes/CrumbViolin_Baguette/M_Emote_CrumbViolin_Baguette_SimpleBlur
graph {
 properties {
  bAutomaticallySetUsageInEditor=false
  bUsedWithNiagaraMeshParticles=true
  bUsedWithNiagaraSprites=true
  blendMode=BLEND_Translucent
  shadingModel=MSM_Unlit
  twoSided=true
  usedShadingModels=Unlit
 }
 nodes [
  n1 = ParticleColor()
  n2 = SphereMask(
   A=TextureCoordinate()
   B=Constant() {r=0.5}
   Radius=ScalarParameter() {defaultValue=0.4, parameterName=Radius, sliderMax=0.5}
   Hardness=ScalarParameter() {defaultValue=0.7, parameterName=Hardness, sliderMax=1}
  ) {attenuationRadius=0.4, hardnessPercent=70}
  n3 = ScalarParameter() {defaultValue=1, parameterName=AlphaMul}
  DepthFade(Opacity=Multiply(A=@n1.A, B=Multiply(A=@n2, B=@n3))) {fadeDistanceDefault=5}
  MaterialFunctionCall(Emissive=Multiply(
    A=@n1.RGB
    B=TextureSampleParameter2D() {parameterName=Tex, texture=&/T_Emote_CrumbViolin_Baguette_TailBlur}.RGB
   )) {materialFunction=&MF_FX_EmissiveAdjustments}
 ]
}
